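# Niagara spit effect for the HighTower Radish pet cosmetic: a brief, one-shot burst of pale, translucent droplets.
# Spit_Drop launches one larger sprite forward along +X with slight lift, while Extra_Spit_Drops adds one or two smaller, slightly slower drops from a tiny sphere.
# Both arc down under gravity over a short lifetime, using size curves, speed-based sprite stretching and alpha fade-out to read as flying liquid.
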NiagaraSystem /BRCosmetics/Effects/Fort_Effects/Effects/Pets/HighTower_Radish/NS_Pet_HighTower_Radish_Spit
dependencies {usedDataInterfaces [&NiagaraDataInterfaceCurve], usedDynamicInputs [], usedModules [&SystemState]}
emitters {
 Extra_Spit_Drops {
  emitterUpdateScript [
   &EmitterState()
   &SpawnBurst_Instantaneous(Spawn Count=&UniformRangedInt(Minimum=1, Maximum=2), Use Spawn Probability=false)
  ]
  particleSpawnScript [
   /Niagara/Modules/Spawn/Initialization/InitializeParticle(
    Color=<rgba 0.9115 1 1 0.85>
    Lifetime=&UniformRangedFloat(Minimum=0.3, Maximum=0.35)
    Module.Write Color=true
    Module.Write Lifetime=true
    Module.Write Mass=true
    Module.Write Material Random=true
    Module.Write Position=false
    Module.Write Scale=false
    Module.Write SpriteRotation=false
    Module.Write SpriteSize=true
    Position=&SimulationPosition()
    Sprite Size=&Vector2DFromFloat(Value=&UniformRangedFloat(Minimum=1.25, Maximum=1.5))
   )
   &AddVelocity(Velocity=&UniformRangedVector(Minimum=<260 0 50>, Maximum=<285 0 50>))
   /Niagara/Modules/Spawn/Location/SphereLocation(MaskSpawnGroup=false, Module.Override Randomness=false, Sphere Radius=1, Surface Only=false)
  ]
  particleUpdateScript [
   &ParticleState()
   &GravityForce(Gravity=<0 0 -800>)
   &SolveForcesAndVelocity(Clamp Velocity=false, Limit Acceleration=false)
   &ScaleSpriteSize(Scale Factor=&Vector2DFromFloat(Value=&FloatFromCurve(FloatCurve=DataInterface() {propertyValues {Curve {keys [
          <at 0 0 RCIM_Cubic>
          {arriveTangent=3.5998, interpMode=RCIM_Cubic, leaveTangent=3.5998, tangentMode=RCTM_User, time=0.012, value=0.7657}
          {arriveTangent=-0.7751, interpMode=RCIM_Cubic, leaveTangent=-0.7751, time=0.35, value=1}
          <at 1 0>
         ]}}})))
   &ScaleSpriteSizeBySpeed(Max Scale Factor=<1 6>, Min Scale Factor=<0.5 1>)
   &ScaleColor(
    Scale Alpha=&FloatFromCurve(FloatCurve=DataInterface() {propertyValues {Curve {keys [<at 0.3 1 RCIM_Cubic>,<at 1 0>]}}})
    ScaleA=true
    ScaleRGB=true
   )
  ]
  rendererClasses [&NiagaraSpriteRendererProperties]
  renderers [{rendererClass=&NiagaraSpriteRendererProperties}]
 }
 Spit_Drop {
  emitterUpdateScript [
   &EmitterState()
   &SpawnBurst_Instantaneous(Spawn Count=1, Use Spawn Probability=false)
  ]
  particleSpawnScript [
   &InitializeParticle(
    Color=<rgba 0.9115 1 1 0.85>
    Lifetime=&UniformRangedFloat(Minimum=0.3, Maximum=0.35)
    Module.Write Color=true
    Module.Write Lifetime=true
    Module.Write Mass=true
    Module.Write Material Random=true
    Module.Write Position=false
    Module.Write Scale=false
    Module.Write SpriteRotation=false
    Module.Write SpriteSize=true
    Position=&SimulationPosition()
    Sprite Size=&Vector2DFromFloat(Value=&UniformRangedFloat(Minimum=1.85, Maximum=2))
   )
   &AddVelocity(Velocity=<325 0 50>)
  ]
  particleUpdateScript [
   &ParticleState()
   &GravityForce(Gravity=<0 0 -800>)
   &SolveForcesAndVelocity(Clamp Velocity=false, Limit Acceleration=false)
   &ScaleSpriteSize(Scale Factor=&Vector2DFromFloat(Value=&FloatFromCurve(FloatCurve=DataInterface() {propertyValues {Curve {keys [
          <at 0 0 RCIM_Cubic>
          {arriveTangent=3.5998, interpMode=RCIM_Cubic, leaveTangent=3.5998, tangentMode=RCTM_User, time=0.012, value=0.7657}
          {arriveTangent=-0.7751, interpMode=RCIM_Cubic, leaveTangent=-0.7751, time=0.35, value=1}
          <at 1 0>
         ]}}})))
   &ScaleSpriteSizeBySpeed(Max Scale Factor=<1 8>)
   &ScaleColor(
    Scale Alpha=&FloatFromCurve(FloatCurve=DataInterface() {propertyValues {Curve {keys [<at 0.3 1 RCIM_Cubic>,<at 1 0>]}}})
    ScaleA=true
    ScaleRGB=true
   )
  ]
  rendererClasses [&NiagaraSpriteRendererProperties]
  renderers [{rendererClass=&NiagaraSpriteRendererProperties}]
 }
}
summary {
 systemName=NS_Pet_HighTower_Radish_Spit
 emitters [
  {emitterName=Spit_Drop, rendererClasses [&NiagaraSpriteRendererProperties]}
  {emitterName=Extra_Spit_Drops, rendererClasses [&NiagaraSpriteRendererProperties]}
 ]
 userVariables []
}
